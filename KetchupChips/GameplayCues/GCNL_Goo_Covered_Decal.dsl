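; Looping gameplay cue that covers a player pawn in dripping goo while active.
; On start it records mesh bounds, clears old FX, then after a delay attaches drip Niagara FX using custom depth, body size and slide/crouch state.
; For locally viewed pawns, async tag listeners hide drips from the owner in first-person aiming.
; Removal fades drips and decal planes with removal FX, or destroys drips immediately when possessing a prop.
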(event Niagara|EventReceiveParticleData
  (params (Data :type Array%20of%20Basic%20Particle%20Data%20Structures) (NiagaraSystem :type Niagara%20System%20Object%20Reference) (SimulationPositionOffset :type Vector%20%28by%20ref%29 :default "0, 0, 0"))
  ;[1 0.3645 0.3744] Getting physical Surface Index from Niagara Particle Collision 2.
  ; This is for large explosion's mud chunks ground collision sound FX.
  (for _array_element Data
    _))

(event EventOnApplication(Cascade)
  (params (MyTarget :type Actor%20Object%20Reference) (Parameters :type Gameplay%20Cue%20Parameters%20Structure%20%28by%20ref%29) (ParticleComponents :type Array%20of%20Cascade%20Particle%20System%20Component%20Object%20References) (AudioComponents :type Array%20of%20Audio%20Component%20Object%20References) (BurstCameraShakeInstance :type Camera%20Shake%20Base%20Object%20Reference) (BurstDecalInstance :type Decal%20Actor%20Object%20Reference)))

(event EventOnLoopingStart(EitherVFX-preferthisversion)
  (params (MyTarget :type Actor%20Object%20Reference) (Parameters :type Gameplay%20Cue%20Parameters%20Structure%20%28by%20ref%29) (ParticleComponents :type Array%20of%20FXSystem%20Component%20Object%20References) (AudioComponents :type Array%20of%20Audio%20Component%20Object%20References))
  (bind _firstpersontag (Variables|Default|GetFirstPersonTag))
  (Utilities|FlowControl|Sequence
    (:then_0
      (bind _asfort_player_pawn (Utilities|Casting|CastToFortPlayerPawn MyTarget))
      (bind _output_get (Variables|Default|SetPlayerPawn _asfort_player_pawn))
      ;[1 1 1] For local pawns, listening for entering and exiting first-person camera (aiming down a mesh scope)
      (if (Pawn|IsLocallyViewed _output_get)
        (bind _asynctaskproxy (Ability|Tasks|WaitGameplayTagAddToActor MyTarget _firstpersontag)
          (:then
            (Variables|Default|SetFirstPersonTagAddedListener _asynctaskproxy)
            (bind _asynctaskproxy_2 (Ability|Tasks|WaitGameplayTagRemoveFromActor MyTarget _firstpersontag)
              (:then
                (Variables|Default|SetFirstPersonTagRemovedListener _asynctaskproxy_2))
              (:Removed
                (CallFunction|OnFirstPersonAimingChanged false))))
          (:Added
            (CallFunction|OnFirstPersonAimingChanged true)))))
    (:then_1
      (Variables|Default|SetLoopingTrailVFX (Utilities|Array|Get(acopy) ParticleComponents))
      (CallFunction|ApplyDecals MyTarget)
      (Utilities|FlowControl|Delay MyTarget (Variables|Default|GetApplicationDelay))
      (CallFunction|ApplyDripFX)
      (bind _spawneddripfx (Variables|Default|GetSpawnedDripFX)
        (:then
          (bind _asplayer_pawn_athena (Utilities|Casting|CastToPlayerPawn_Athena MyTarget))
          (Effects|Components|ParticleSystem|SetBoolParameter _spawneddripfx "IsSliding" (Variables|/Script/FortniteGame.FortPawn|GetbIsSliding _asplayer_pawn_athena))
          (Effects|Components|ParticleSystem|SetBoolParameter _spawneddripfx "IsCrouched" (Variables|/Script/Engine.Character|GetbIsCrouched _asplayer_pawn_athena)))
        (:else)))))

(event EventOnRemoval(EitherVFX-preferthisversion)
  (params (MyTarget :type Actor%20Object%20Reference) (Parameters :type Gameplay%20Cue%20Parameters%20Structure%20%28by%20ref%29) (ParticleComponents :type Array%20of%20FXSystem%20Component%20Object%20References) (AudioComponents :type Array%20of%20Audio%20Component%20Object%20References) (BurstCameraShakeInstance :type Camera%20Shake%20Base%20Object%20Reference) (BurstDecalInstance :type Decal%20Actor%20Object%20Reference))
  (Utilities|FlowControl|Sequence
    (:then_0
      (bind _firstpersontagaddedlistener (Variables|Default|GetFirstPersonTagAddedListener)
        (:then
          (Ability|Async|EndAction _firstpersontagaddedlistener))
        (:else
          _))
      (bind _firstpersontagremovedlistener (Variables|Default|GetFirstPersonTagRemovedListener)
        (:then
          (Ability|Async|EndAction _firstpersontagremovedlistener))
        (:else)))
    (:then_1
      (Utilities|FlowControl|DelayUntilNextTick 0)
      (bind _asfort_player_pawn (Utilities|Casting|CastToFortPlayerPawn MyTarget))
      ;[1 1 1] When mud is removed as a result of possessing a prop, it looks bad to see the mud dripping off of a skeleton that is no longer visible, so in that case we immediately destroy the FX
      (if (GameplayTags|HasMatchingGameplayTag _asfort_player_pawn (Variables|Default|GetPossessingPropTag))
        (bind _spawneddripfx (Variables|Default|GetSpawnedDripFX)
          (:then
            (Components|DestroyComponent _spawneddripfx 0))
          (:else))
        (else
          (Variables|Default|GetSpawnedDripFX
            (:then
              _)
            (:else
              _))
          (Effects|SpawnFXSystemAttached "/KetchupChips/Assets/FX/NS_KetchupChip_GooRemoval.NS_KetchupChip_GooRemoval" (Variables|/Script/Engine.Character|GetMesh _asfort_player_pawn))
          (bind _asniagara_particle_system_component (Utilities|Casting|CastToNiagaraComponent (Variables|Default|GetSpawnedDripFX)))
          (Niagara|SetNiagaraVariable(Bool) _asniagara_particle_system_component "FadeOut" true)
          ;[0.2465 0.5337 1] Send Parameters to Niagara Mud FX. for body Decal control. Removing
          (Niagara|SetNiagaraVariable(Vector4) _asniagara_particle_system_component "InputFromBP1" (Math|Vector4|MakeVector4 (Utilities|Time|GetGameTimeinSeconds 0) 0.2 (Utilities|Time|GetGameTimeinSeconds 0) 2.0))
          (CallFunction|FadeOut)
          (Utilities|FlowControl|Delay MyTarget (+ (Variables|Default|GetTransitionOut) 0.1))
          (CallFunction|RemoveExistingFX))))))

(event Custom|ApplyDecals
  (params (Object :type Object%20Reference))
  (bind _asplayer_pawn_athena (Utilities|Casting|CastToPlayerPawn_Athena Object))
  (Utilities|IsValid _asplayer_pawn_athena
    (:Is%20Valid
      (bind _output_get (Variables|Default|SetPawn _asplayer_pawn_athena))
      (Variables|Default|SetPawnSkeletalMesh (Variables|/Script/Engine.Character|GetMesh _output_get))
      (bind (_origin _boxextent) (Collision|GetComponentBounds (Variables|/Script/Engine.Character|GetMesh _output_get)))
      (bind (_boxextent_x _boxextent_y _boxextent_z) (Math|Vector|BreakVector _boxextent))
      (Variables|Default|SetPlayerHeight (* _boxextent_z 2.0))
      (bind _result (Math|Float|NegateFloat _boxextent_z)
        (:%20%20
          (Variables|Default|SetPlayerHalfHeightNegated _result)
          (CallFunction|RemoveExistingFX))))))

(event Custom|OnFirstPersonAimingChanged
  (params (bIsFirstPerson :type Boolean))
  (bind _spawneddripfx (Variables|Default|GetSpawnedDripFX))
  ;[1 1 1] Hide Mud FX when ADS with a first person camera - fix for mud clumps obscuring the camera for true first person scopes
  (Utilities|IsValid _spawneddripfx
    (:Is%20Valid
      (Rendering|SetOwnerNoSee _spawneddripfx bIsFirstPerson :!class /Script/Engine.PrimitiveComponent))))

(fn ApplyDripFX
  (bind _playerpawn (Variables|Default|GetPlayerPawn))
  (bind _returnvalue (Cosmetics|GetPawnCosmeticLoadout _playerpawn false true))
  (bind (_character _charactervariantchannels) (Utilities|Struct|BreakFortAthenaLoadout _returnvalue))
  (bind (_outbodytype _returnvalue_2) (Class|AthenaCharacterItemDefinition|FindBodyType _character _charactervariantchannels))
  (bind _returnvalue_3 (Effects|SpawnFXSystemAttached (Variables|Default|GetDripFXAsset) (Variables|/Script/Engine.Character|GetMesh (Variables|Default|GetPlayerPawn))))
  ;[1 0.3645 0.3744] Getting physical Surface Index from Niagara Particle Collision 1
  (bind _asniagara_particle_system_component (Utilities|Casting|CastToNiagaraComponent _returnvalue_3))
  (Rendering|SetRenderCustomDepth _asniagara_particle_system_component true)
  ;;[1 1 1] Custom Depth
  (Rendering|SetCustomDepthStencilValue _asniagara_particle_system_component 22)
  (Niagara|SetNiagaraVariable(Object) _asniagara_particle_system_component "BP_Hit" self)
  ;[0 1 0.3692] Body Decal Material Control
  (CallFunction|SetCustomPrimitiveVariables 0 _asniagara_particle_system_component)
  (Variables|Default|SetSpawnedDripFX _returnvalue_3)
  (if _returnvalue_2
    (Effects|Components|ParticleSystem|SetIntParameter _returnvalue_3 "CharacterSizeID" (Math|Conversions|ToInteger(Byte) _outbodytype))))

(fn FadeOut
  (for _array_element (Variables|Default|GetPlanes)
    (Rendering|Material|SetCustomPrimitiveDataFloat _array_element 2 (Utilities|Time|GetGameTimeinSeconds 0))))

(fn RemoveExistingFX
  (|CallParentFunction)
  (bind _spawneddripfx (Variables|Default|GetSpawnedDripFX)
    (:then
      (Components|DestroyComponent _spawneddripfx 0))
    (:else)))

(fn ConstructionScript
  (|CallParentFunction))
